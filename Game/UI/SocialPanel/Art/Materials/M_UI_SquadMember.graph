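# Translucent UI material for the squad member icon in the social panel, drawn from what appears to be a distance-field texture.
# A multiplied, clamped sample forms the sharp icon shape, while an inverted SmoothStep around a thickness threshold, with adjustable edge softness, cuts its interior.
# A dot product combines the two into an outline, and the UsesOutline scalar lerps from the solid shape, its default, to that outline.
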Material /Game/UI/SocialPanel/Art/Materials/M_UI_SquadMember
graph {
 properties {
  bUsedWithStaticMesh=false
  blendMode=BLEND_Translucent
  materialDomain=MD_UI
  pixelDepthOffsetMode=PDOM_Legacy
  usedShadingModels=Default Lit
 }
 nodes [
  Constant() {r=1}
  n1 = TextureSampleParameter2D() {parameterName=Param, texture=/Game/UI/SocialPanel/Art/Textures/T_UI_DF_SquadMember}
  n2 = Clamp(None=Multiply(A=@n1.RGB, B=ScalarParameter() {defaultValue=32, parameterName=IconSharpnessMultiply}) {constB=32})
  n3 = ScalarParameter() {defaultValue=0.075, parameterName=IconEdgeSharpness}
  n4 = ScalarParameter() {defaultValue=0.5, parameterName=IconEdgeThickness, sliderMax=1}
  n5 = MaterialFunctionCall(Alpha=@n1.RGB, Min=Add(A=Multiply(A=@n3) {constB=-1}, B=@n4), Max=Add(A=@n3, B=@n4)) {materialFunction=&SmoothStep}
  LinearInterpolate(A=@n2, B=DotProduct(A=@n2, B=OneMinus(None=@n5.Result)), Alpha=ScalarParameter() {parameterName=UsesOutline})
 ]
}
