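# Frontend activity browser badge (FortActiveFriendsViolator subclass) showing an active-friends count beside a social icon on a yellow box.
# CallOut Surface plays the Intro animation to pop the badge in, and CallOut Hide plays it in reverse.
# Update-Up and Update-Down pulse scale and color, with Up adding a flare and floating "+1".
# UpdateData's Update-Up nodes are disconnected from its entry, and no graph here plays Update-Down or Reset-Off.
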WidgetBlueprint /FortniteCoreUI/Athena/UI/Frontend/ActivityBrowser/Violator_ActiveFriends
animations [
 Intro {
  binding_count=1
  length_seconds=0.15
  object=&self:Intro
  sequence {bindings [Overlay_Transform {
     id=#1
     possessed_class=Overlay
     tracks [
      {
       type=MovieScene2DTransformTrack
       sections [{
         type=MovieScene2DTransformSection
         end_frame=3
         start_frame=0
         channels [
          Translation.X {type=MovieSceneScriptingFloatChannel, const [0,3]}
          Translation.Y {type=MovieSceneScriptingFloatChannel, keys [[0,48],[2,0],[3,0]]}
          Angle {type=MovieSceneScriptingFloatChannel}
          Scale.X {type=MovieSceneScriptingFloatChannel, keys [[0,0.25],[2,1],[3,1]]}
          Scale.Y {type=MovieSceneScriptingFloatChannel, keys [[0,2],[2,1],[3,1]]}
          Shear.X {type=MovieSceneScriptingFloatChannel}
          Shear.Y {type=MovieSceneScriptingFloatChannel}
         ]
        }]
      }
      {
       type=MovieSceneFloatTrack
       display_name=Render Opacity
       sections [{type=MovieSceneFloatSection, end_frame=3, start_frame=0, channels [None {type=MovieSceneScriptingFloatChannel, keys [[0,0],[1.8,1],[3,1]]}]}]
      }
     ]
    }]}
 }
 Update-Up {
  binding_count=4
  length_seconds=0.25
  object=&self:Update-Up
  sequence {bindings [
    Image_UpdateFlare {
     id=#2
     possessed_class=Image
     tracks [
      {
       type=MovieSceneByteTrack
       display_name=Visibility
       sections [{
         type=MovieSceneByteSection
         end_frame=5
         start_frame=0
         channels [None {type=MovieSceneScriptingByteChannel, keys [[0,1],[0.2,3],[4.8,3],[5,1]]}]
        }]
      }
      {
       type=MovieSceneFloatTrack
       display_name=Render Opacity
       sections [{
         type=MovieSceneFloatSection
         end_frame=5
         start_frame=0
         channels [None {type=MovieSceneScriptingFloatChannel, keys [[0.2,0],[2,0.35],[4.8,0]]}]
        }]
      }
      {
       type=MovieScene2DTransformTrack
       sections [{
         type=MovieScene2DTransformSection
         end_frame=5
         start_frame=0
         channels [
          Translation.X {type=MovieSceneScriptingFloatChannel, const [18,2]}
          Translation.Y {type=MovieSceneScriptingFloatChannel, keys [[0,-24],[4,-32]]}
          Angle {type=MovieSceneScriptingFloatChannel}
          Scale.X {type=MovieSceneScriptingFloatChannel}
          Scale.Y {type=MovieSceneScriptingFloatChannel}
          Shear.X {type=MovieSceneScriptingFloatChannel, const [-8,1]}
          Shear.Y {type=MovieSceneScriptingFloatChannel}
         ]
        }]
      }
     ]
    }
    Text_DeltaChange {
     id=#3
     possessed_class=CommonTextBlock
     tracks [
      {
       type=MovieScene2DTransformTrack
       sections [{
         type=MovieScene2DTransformSection
         end_frame=5
         start_frame=0
         channels [
          Translation.X {type=MovieSceneScriptingFloatChannel, const [18,2]}
          Translation.Y {type=MovieSceneScriptingFloatChannel, keys [[0,-24],[4,-32]]}
          Angle {type=MovieSceneScriptingFloatChannel}
          Scale.X {type=MovieSceneScriptingFloatChannel}
          Scale.Y {type=MovieSceneScriptingFloatChannel}
          Shear.X {type=MovieSceneScriptingFloatChannel}
          Shear.Y {type=MovieSceneScriptingFloatChannel}
         ]
        }]
      }
      {
       type=MovieSceneFloatTrack
       display_name=Render Opacity
       sections [{type=MovieSceneFloatSection, end_frame=5, start_frame=0, channels [None {type=MovieSceneScriptingFloatChannel, keys [[2,1],[4.8,0],[5,0]]}]}]
      }
      {
       type=MovieSceneByteTrack
       display_name=Visibility
       sections [{
         type=MovieSceneByteSection
         end_frame=5
         start_frame=0
         channels [None {type=MovieSceneScriptingByteChannel, keys [[0,1],[0.2,3],[4.8,3],[5,1]]}]
        }]
      }
     ]
    }
    HorizontalBox_0 {
     id=#4
     possessed_class=HorizontalBox
     tracks [{
       type=MovieScene2DTransformTrack
       sections [{
         type=MovieScene2DTransformSection
         end_frame=5
         start_frame=0
         channels [
          Translation.X {type=MovieSceneScriptingFloatChannel}
          Translation.Y {type=MovieSceneScriptingFloatChannel}
          Angle {type=MovieSceneScriptingFloatChannel}
          Scale.X {type=MovieSceneScriptingFloatChannel, keys [[0,1],[0.8,1.3],[2.8,1]]}
          Scale.Y {type=MovieSceneScriptingFloatChannel, keys [[0,1],[0.8,1.3],[2.8,1]]}
          Shear.X {type=MovieSceneScriptingFloatChannel}
          Shear.Y {type=MovieSceneScriptingFloatChannel}
         ]
        }]
      }]
    }
    YellowBox {
     id=#5
     possessed_class=Image
     tracks [
      {
       type=MovieScene2DTransformTrack
       sections [{
         type=MovieScene2DTransformSection
         end_frame=5
         start_frame=0
         channels [
          Translation.X {type=MovieSceneScriptingFloatChannel}
          Translation.Y {type=MovieSceneScriptingFloatChannel}
          Angle {type=MovieSceneScriptingFloatChannel}
          Scale.X {type=MovieSceneScriptingFloatChannel, keys [[0,1],[0.8,1.1],[2.8,1]]}
          Scale.Y {type=MovieSceneScriptingFloatChannel, keys [[0,1],[0.8,1.1],[2.8,1]]}
          Shear.X {type=MovieSceneScriptingFloatChannel}
          Shear.Y {type=MovieSceneScriptingFloatChannel}
         ]
        }]
      }
      {
       type=MovieSceneWidgetMaterialTrack
       sections [{
         type=MovieSceneParameterSection
         end_frame=5
         start_frame=0
         channels [
          RGBAR {type=MovieSceneScriptingFloatChannel, const [1,3]}
          RGBAG {type=MovieSceneScriptingFloatChannel, thinned_from=3, keys [[0,0.9992],[5,0.9992]]}
          RGBAB {type=MovieSceneScriptingFloatChannel, keys [[0,0],[2,1],[5,0]]}
          RGBAA {type=MovieSceneScriptingFloatChannel, const [1,3]}
         ]
        }]
      }
     ]
    }
   ]}
 }
 Update-Down {
  binding_count=2
  length_seconds=0.25
  object=&self:Update-Down
  sequence {bindings [
    HorizontalBox_0 {
     id=#4
     possessed_class=HorizontalBox
     tracks [{
       type=MovieScene2DTransformTrack
       sections [{
         type=MovieScene2DTransformSection
         end_frame=5
         start_frame=0
         channels [
          Translation.X {type=MovieSceneScriptingFloatChannel}
          Translation.Y {type=MovieSceneScriptingFloatChannel}
          Angle {type=MovieSceneScriptingFloatChannel}
          Scale.X {type=MovieSceneScriptingFloatChannel, keys [[0,1],[0.8,0.9],[2.8,1]]}
          Scale.Y {type=MovieSceneScriptingFloatChannel, keys [[0,1],[0.8,0.9],[2.8,1]]}
          Shear.X {type=MovieSceneScriptingFloatChannel}
          Shear.Y {type=MovieSceneScriptingFloatChannel}
         ]
        }]
      }]
    }
    YellowBox {
     id=#5
     possessed_class=Image
     tracks [
      {
       type=MovieScene2DTransformTrack
       sections [{
         type=MovieScene2DTransformSection
         end_frame=5
         start_frame=0
         channels [
          Translation.X {type=MovieSceneScriptingFloatChannel}
          Translation.Y {type=MovieSceneScriptingFloatChannel}
          Angle {type=MovieSceneScriptingFloatChannel}
          Scale.X {type=MovieSceneScriptingFloatChannel, keys [[0,1],[1,0.95],[2.8,1]]}
          Scale.Y {type=MovieSceneScriptingFloatChannel, keys [[0,1],[1,0.95],[2.8,1]]}
          Shear.X {type=MovieSceneScriptingFloatChannel}
          Shear.Y {type=MovieSceneScriptingFloatChannel}
         ]
        }]
      }
      {
       type=MovieSceneWidgetMaterialTrack
       sections [{
         type=MovieSceneParameterSection
         end_frame=5
         start_frame=0
         channels [
          RGBAR {type=MovieSceneScriptingFloatChannel, keys [[0,1],[2,0.75],[5,1]]}
          RGBAG {type=MovieSceneScriptingFloatChannel, keys [[0,0.9992],[2,0.7492],[5,0.9992]]}
          RGBAB {type=MovieSceneScriptingFloatChannel, const [0,3]}
          RGBAA {type=MovieSceneScriptingFloatChannel, keys [[0,1],[2,0.75],[5,1]]}
         ]
        }]
      }
     ]
    }
   ]}
 }
 Reset-Off {
  binding_count=1
  length_seconds=0.01
  object=&self:Reset-Off
  sequence {bindings [Overlay_Transform {
     id=#1
     possessed_class=Overlay
     tracks [
      {
       type=MovieScene2DTransformTrack
       sections [{
         type=MovieScene2DTransformSection
         end_frame=0
         start_frame=0
         channels [
          Translation.X {type=MovieSceneScriptingFloatChannel, const [0,1]}
          Translation.Y {type=MovieSceneScriptingFloatChannel, const [48,1]}
          Angle {type=MovieSceneScriptingFloatChannel}
          Scale.X {type=MovieSceneScriptingFloatChannel, const [0.25,1]}
          Scale.Y {type=MovieSceneScriptingFloatChannel, const [2,1]}
          Shear.X {type=MovieSceneScriptingFloatChannel}
          Shear.Y {type=MovieSceneScriptingFloatChannel}
         ]
        }]
      }
      {
       type=MovieSceneFloatTrack
       display_name=Render Opacity
       sections [{type=MovieSceneFloatSection, end_frame=0, start_frame=0, channels [None {type=MovieSceneScriptingFloatChannel, const [0,1]}]}]
      }
     ]
    }]}
 }
]
graphs [
 CallOut Hide {nodes [
   CallOut Hide() {is=title K2Node_FunctionEntry}
   > PlayAnimation(InAnimation=Get Intro(), PlayMode=Reverse)
  ]}
 CallOut Surface {nodes [
   CallOut Surface() {is=title K2Node_FunctionEntry}
   > PlayAnimation(InAnimation=Get Intro())
  ]}
 EventGraph {nodes [
   Event PreConstruct()
   Event Construct()
  ]}
 UpdateData {nodes [
   "SetText (Text)"(self=Get Text_DeltaChange() -> (Text_DeltaChange:Common Text Object Reference), InText="INVTEXT(\"+1\")")
   > PlayAnimation(InAnimation=Get Update-Up() -> (Update-Up:Widget Animation Object Reference))
   UpdateData() {is=title K2Node_FunctionEntry}
  ]}
]
variables []
widgets {
 info {inheritedWidgetCount=1, parentClass=/Script/FortniteUI.FortActiveFriendsViolator, rootWidgetClass=&Overlay, widgetCount=7}
 widgets [
  Overlay_Transform {widgetClassPath=&Overlay, widget=Overlay_Transform}
  YellowBox {
   widgetClassPath=&Image
   bIsVariable=true
   parent=Overlay_Transform
   widget=YellowBox
   properties {brush {
     bIsDynamicallyLoaded=false
     drawAs=Box
     imageType=NoImage
     mirroring=NoMirror
     resourceName=None
     resourceObject=/Game/Athena/UI/Frontend/ActivityBrowser/Art/M_UI_GameMode_SocialBox
     tiling=NoTile
     imageSize=<32 32>
     margin=<ltrb 0.5 0 0.5 0>
     outlineSettings {
      bUseBrushTransparency=false
      roundingType=HalfHeightRadius
      width=0
      color {colorUseRule=UseColor_Specified, specifiedColor=<rgba 0 0 0 0>}
      cornerRadii=<0 0 0 0>
     }
     tintColor {colorUseRule=UseColor_Specified, specifiedColor=<rgba 1 1 1 1>}
     uVRegion {bIsValid=false, max=<0 0>, min=<0 0>}
    }}
   slot {class=OverlaySlot, properties {horizontalAlignment=HAlign_Fill}}
  }
  HorizontalBox_0 {
   widgetClassPath=&HorizontalBox
   parent=Overlay_Transform
   widget=HorizontalBox_0
   slot {class=OverlaySlot, properties {verticalAlignment=VAlign_Center, padding=<ltrb 8 0 8 2>}}
  }
  SocialIcon {
   widgetClassPath=&Image
   bIsVariable=true
   parent=HorizontalBox_0
   widget=SocialIcon
   properties {brush {
     bIsDynamicallyLoaded=false
     drawAs=Image
     imageType=NoImage
     mirroring=NoMirror
     resourceName=None
     resourceObject=/Game/UI/SocialPanel/Art/Materials/M_UI_SquadMember
     tiling=NoTile
     imageSize=<22 22>
     margin=<ltrb 0 0 0 0>
     outlineSettings {
      bUseBrushTransparency=false
      roundingType=HalfHeightRadius
      width=0
      color {colorUseRule=UseColor_Specified, specifiedColor=<rgba 0 0 0 0>}
      cornerRadii=<0 0 0 0>
     }
     tintColor {colorUseRule=UseColor_Specified, specifiedColor=<rgba 0 0 0 1>}
     uVRegion {bIsValid=false, max=<0 0>, min=<0 0>}
    }}
   slot {class=HorizontalBoxSlot, properties {horizontalAlignment=HAlign_Center, verticalAlignment=VAlign_Center, padding=<ltrb 0 0 0 1>}}
  }
  Text_ActivePlayerCount {
   widgetClassPath=&CommonTextBlock
   bInherited=true
   bIsVariable=true
   parent=HorizontalBox_0
   widget=Text_ActivePlayerCount
   properties {
    text=99
    colorAndOpacity {colorUseRule=UseColor_Specified, specifiedColor=<rgba 0 0 0 1>}
    font {
     bForceMonospaced=false
     bMaterialIsStencil=false
     fontMaterial=null
     fontObject=&BurbankSmall
     letterSpacing=0
     monospacedWidth=1
     size=18
     skewAmount=0
     typefaceFontName=Black
     outlineSettings {
      bApplyOutlineToDropShadows=false
      bMiteredCorners=false
      bSeparateFillAlpha=false
      outlineMaterial=null
      outlineSize=0
      outlineColor=<rgba 0 0 0 1>
     }
    }
    renderTransform {angle=0, scale=<1 1>, shear=<-8 0>, translation=<0 0>}
    shadowOffset=<0 0>
   }
   slot {class=HorizontalBoxSlot, properties {horizontalAlignment=HAlign_Center, verticalAlignment=VAlign_Center}}
  }
  Image_UpdateFlare {
   widgetClassPath=&Image
   bIsVariable=true
   parent=Overlay_Transform
   widget=Image_UpdateFlare
   properties {
    visibility=Collapsed
    brush {
     bIsDynamicallyLoaded=false
     drawAs=Image
     imageType=NoImage
     mirroring=NoMirror
     resourceName=None
     resourceObject=/Game/Athena/UI/Frontend/ActivityBrowser/Art/M_UI_Social_UpdateFlare
     tiling=NoTile
     imageSize=<32 32>
     margin=<ltrb 0 0 0 0>
     outlineSettings {
      bUseBrushTransparency=false
      roundingType=HalfHeightRadius
      width=0
      color {colorUseRule=UseColor_Specified, specifiedColor=<rgba 0 0 0 0>}
      cornerRadii=<0 0 0 0>
     }
     tintColor {colorUseRule=UseColor_Specified, specifiedColor=<rgba 1 1 0.25 1>}
     uVRegion {bIsValid=false, max=<0 0>, min=<0 0>}
    }
    renderTransform {angle=0, scale=<1.5 1.5>, shear=<0 0>, translation=<0 0>}
   }
   slot {class=OverlaySlot, properties {horizontalAlignment=HAlign_Center, verticalAlignment=VAlign_Center}}
  }
  Text_DeltaChange {
   widgetClassPath=&CommonTextBlock
   bIsVariable=true
   parent=Overlay_Transform
   widget=Text_DeltaChange
   properties {
    text=+1
    visibility=Collapsed
    font {
     bForceMonospaced=false
     bMaterialIsStencil=false
     fontMaterial=null
     fontObject=&BurbankSmall
     letterSpacing=0
     monospacedWidth=1
     size=18
     skewAmount=0
     typefaceFontName=Black
     outlineSettings {
      bApplyOutlineToDropShadows=false
      bMiteredCorners=false
      bSeparateFillAlpha=false
      outlineMaterial=null
      outlineSize=0
      outlineColor=<rgba 0 0 0 1>
     }
    }
    renderTransform {angle=0, scale=<1 1>, shear=<-8 0>, translation=<18 -24>}
    shadowOffset=<0 0>
   }
   slot {class=OverlaySlot, properties {horizontalAlignment=HAlign_Center, verticalAlignment=VAlign_Center}}
  }
 ]
}
